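# Unlit two-sided alpha-composite material for Niagara sprites that draws a tweakable, spinning, eroding noise swirl.
# UVs rotate by radial Twist plus SpinSpeed over time, then tile, pan and pick up normal-map distortion to sample two noise textures.
# A SphereMask folded into a soft ring shapes the noise, which blends ColorA and ColorB beneath added emissive noise.
# The particle Erosion dynamic parameter cuts the shape, with DepthFade softening intersections.
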Material /Game/Effects/Fort_Effects/Materials/Swirl/M_Swirl_Tweakable
graph {
 properties {
  bAutomaticallySetUsageInEditor=false
  bUsedWithNiagaraSprites=true
  blendMode=BLEND_AlphaComposite
  shadingModel=MSM_Unlit
  twoSided=true
  usedShadingModels=Unlit
 }
 nodes [
  MaterialFunctionCall() {materialFunction=&DebugTimeSine}
  n1 = VectorParameter() {parameterName=ColorA, defaultValue=<rgba 0.0404 0.4638 1 1>}
  n2 = TextureSampleParameter2D(UVs=Multiply(A=NamedRerouteUsage(), B=Constant2Vector() {g=1, r=5})) {
   parameterName=DistortionMap
   samplerType=SAMPLERTYPE_Normal
   texture=/Game/Characters/Player/Female/Medium/Bodies/F_MED_Bistro_Astronaut/FX/T_GaussianNoise_Normal
  }
  n3 = Add(A=Divide(A=@n2.R, B=ScalarParameter() {defaultValue=15, parameterName=DetailDistortion}) {constB=5}, B=NamedRerouteUsage())
  n4 = Constant() {r=0.5}
  n5 = SphereMask(A=TextureCoordinate(), B=@n4, Radius=@n4) {hardnessPercent=0}
  n6 = Abs(None=Subtract(A=Multiply(A=@n5) {constB=2}))
  n7 = Multiply(
   A=TextureSampleParameter2D(UVs=@n3) {parameterName=BasicNoiseMap, texture=/Game/Effects/Fort_Effects/Emotes/UltraLight/Texture/T_SoftNoise_BC7}.B
   B=SmoothStep(Value=OneMinus(None=@n6)) {constMin=0.2}
  )
  n8 = LinearInterpolate(A=VectorParameter() {parameterName=ColorB, defaultValue=<rgba 0.0002 0.0145 0.0417 1>}.RGB, B=@n1.RGB, Alpha=@n7)
  n9 = Multiply(
   A=@n1.RGB
   B=TextureSampleParameter2D(UVs=@n3) {parameterName=AdditionalEmissiveNoiseMap, texture=/Game/Effects/Fort_Effects/Textures/Tile/T_Noise_03}.RGB
  )
  n10 = OneMinus(None=DynamicParameter() {paramNames [Erosion,Param2,Param3,Param4]}.Erosion)
  n11 = Multiply(A=Add(A=Divide(A=@n8), B=Multiply(A=@n9) {constB=3}), B=Subtract(A=@n7, B=@n10))
  MaterialFunctionCall(Emissive=DepthFade(Opacity=Saturate(None=@n11))) {materialFunction=&MF_FX_EmissiveAdjustments}
  Multiply(A=DepthFade(Opacity=Subtract(A=@n7, B=@n10)), B=ScalarParameter() {defaultValue=1, parameterName=OpacityMultiplier})
  n12 = ConstantBiasScale(None=TextureCoordinate()) {bias=-0.5, scale=2}
  n13 = Multiply(A=OneMinus(None=Length(None=@n12) {is=class MaterialExpressionLength}), B=ScalarParameter() {defaultValue=-5, parameterName=Twist})
  n14 = Multiply(A=ScalarParameter() {defaultValue=-5, parameterName=SpinSpeed}, B=Time())
  MaterialExpressionMultiply_5 = Multiply(
   A=Rotator(Coordinate=TextureCoordinate(), Time=Add(A=@n13, B=@n14)) {speed=1}
   B=AppendVector(A=ScalarParameter() {defaultValue=1, parameterName=U_Tile}, B=ScalarParameter() {defaultValue=2, parameterName=V_Tile})
  )
  n15 = Convert(None=@MaterialExpressionMultiply_5) {
   nodeName=Break Float2
   convertInputs [{
     type=Vector2
     expressionInput {
      expression=&self:MaterialExpressionMultiply_5
      inputName=None
      mask=0
      maskA=0
      maskB=0
      maskG=0
      maskR=0
      outputIndex=0
     }
    }]
   convertOutputs [{type=Scalar}, {type=Scalar}]
  }
  n16 = Time()
  n17 = Multiply(A=@n16, B=ScalarParameter() {defaultValue=0.5, parameterName=NoisePanSPeedU})
  n18 = Multiply(A=@n16, B=ScalarParameter() {defaultValue=0.5, parameterName=NoisePanSPeedV})
  NamedRerouteDeclaration(None=AppendVector(A=Add(A=@n15, B=@n17), B=Add(A=@n15, B=@n18))) {name=UV, nodeColor=<rgba 1 0 0.0075 1>}
 ]
}
